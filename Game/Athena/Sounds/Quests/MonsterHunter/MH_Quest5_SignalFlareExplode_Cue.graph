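# Sound cue for the signal flare explosion in Monster Hunter quest 5, built from CandyApple fireworks emote explosion samples.
# A Random node picks one of six equally weighted waves and is excluded from branch culling, then a Modulator applies slight pitch and volume variation.
# It plays in the quests sound class with emote attenuation and a concurrency override limiting simultaneous instances.
# It also has a submix send to the music audio analysis submix.
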SoundCue /Game/Athena/Sounds/Quests/MonsterHunter/MH_Quest5_SignalFlareExplode_Cue
root=Modulator_0
nodes [
 {id=Modulator_0, children [Random_0], properties {pitchMax=1.1, pitchMin=0.9, volumeMax=1, volumeMin=0.8}}
 {
  id=Random_0
  children [WavePlayer_0,WavePlayer_2,WavePlayer_3,WavePlayer_5,WavePlayer_6,WavePlayer_7]
  properties {bShouldExcludeFromBranchCulling=true, weights [1,1,1,1,1,1]}
 }
 {id=WavePlayer_0, properties {soundWaveAssetPtr=/Game/Athena/Sounds/Emotes/CandyApple_Fireworks/CA_Emote_Fireworks_Explode_01}}
 {id=WavePlayer_2, properties {soundWaveAssetPtr=/Game/Athena/Sounds/Emotes/CandyApple_Fireworks/CA_Emote_Fireworks_Explode_03}}
 {id=WavePlayer_3, properties {soundWaveAssetPtr=/Game/Athena/Sounds/Emotes/CandyApple_Fireworks/CA_Emote_Fireworks_Explode_05}}
 {id=WavePlayer_5, properties {soundWaveAssetPtr=/Game/Athena/Sounds/Emotes/CandyApple_Fireworks/CA_Emote_Fireworks_Explode_10}}
 {id=WavePlayer_6, properties {soundWaveAssetPtr=/Game/Athena/Sounds/Emotes/CandyApple_Fireworks/CA_Emote_Fireworks_Explode_06}}
 {id=WavePlayer_7, properties {soundWaveAssetPtr=/Game/Athena/Sounds/Emotes/CandyApple_Fireworks/CA_Emote_Fireworks_Explode_08}}
]
properties {
 attenuationSettings=&Fort_Emotes_Athena_Att
 bOverrideConcurrency=true
 maxDistance=1600
 soundClassObject=&Fort_Quests
 volumeMultiplier=1
 concurrencyOverrides {maxCount=3}
 soundSubmixSends [{minSendLevel=0, soundSubmix=&Fortnite_AudioAnalysis_Music}]
}
